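# Post-process Animation Blueprint for the VastGeese biped cosmetic companion, layering pose fixes and secondary motion onto the incoming pose.
# Outside emotes it remaps curves from the mesh, applies the pose asset and adjusts the pelvis against foot IK bones; the `isemote` curve switches to the pose asset alone.
# AnimDynamics then springs ears and antenna, and additive rotations offset both upper arms.
# Updates also set accessory and dance-loop flags, unused in this graph.
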AnimBlueprint /CosmeticCompanions/Assets/Biped/VastGeese/Meshes/VastGeese_Biped_Post_AnimBP
target_skeleton=&/VastGeese_Biped_Skeleton
graphs [
 AnimGraph {nodes [
   n1 = Remap Curves From Mesh!(
    SourcePose:Pose Link Structure=Use cached pose 'StartPose'()
    CurveExpressionsDataAsset:Curve Expressions Data Asset Object Reference=&/VastGeese_EyeSetup_Remap
    SourceMeshComponent:Skeletal Mesh Component Weak object=Get mesh component() -> (mesh component:Skeletal Mesh Component Object Reference)
   ) -> (Pose:Pose Link Structure) {is=title AnimGraphNode_RemapCurvesFromMesh, bUseAttachedParent=true, expressionSource=DataAsset}
   n2 = Get PoseAsset() -> (PoseAsset:Pose Asset Object Reference)
   n3 = "Constraint - pelvis (PP)"!(
    ComponentPose=Local To Component(LocalPose=Pose!(PoseAsset:Pose Asset Object Reference=@n2, SourcePose:Pose Link Structure=@n1) -> (Pose:Pose Link Structure) {is=title AnimGraphNode_PoseBlendNode, blendOption=Linear})
    Alpha=1
    ConstraintWeights_0=0.5
    ConstraintWeights_1:float=0.5
   ) -> (Pose) {
    boneToModify {boneName=pelvis}
    constraintSetup [{targetBone {boneName=IK_Foot_L}}, {targetBone {boneName=IK_Foot_R}}]
    constraintWeights [1,1]
   }
   n4 = Component To Local(ComponentPose="Transform (Modify) Bone - Bone: pelvis"!(ComponentPose=@n3, bAlphaBoolEnabled:bool=True, Translation="0,0,2") -> (Pose) {alphaInputType=Bool, translationMode=BMM_Additive, boneToModify {boneName=pelvis}, translation {z=2}})
   n5 = Pose(PoseAsset=Get PoseAsset(), SourcePose=Use cached pose 'StartPose'()) {blendOption=Linear}
   n6 = Get isEmote() -> (isEmote:bool)
   n7 = Anim Dynamics - Bone: ear_l(
    ComponentPose=Local To Component(LocalPose=Blend Poses by bool(BlendPose_0=@n5, BlendPose_1=@n4, BlendTime_0=0.2, BlendTime_1=0.2, bActiveValue=@n6))
    Alpha=1
    ExternalForce="0,0,0"
   ) {
    angularDampingOverride=1
    angularSpringConstant=10000
    bAngularSpring=true
    bOverrideAngularDamping=true
    bUseGravityOverride=true
    gravityScale=0.35
    simulationSpace=RootRelative
    alphaScaleBiasClamp {bias=1, scale=-1}
    boundBone {boneName=ear_l}
    componentAppliedLinearAccClamp=<10000 10000 10000>
    componentLinearAccScale=<1 1 1>
    componentLinearVelScale=<1 1 1>
    gravityOverride {z=-980}
    physicsBodyDefinitions [{
      boundBone {boneName=ear_l}
      boxExtents=<20 30 30>
      constraintSetup {angularLimitsMax=<5 10 5>, angularLimitsMin=<-5 -10 -5>, angularTarget=<1 0 0>}
      localJointOffset=<4 0 0>
     }]
   }
   n8 = Anim Dynamics - Bone: ear_r(ComponentPose=@n7, Alpha=1, ExternalForce="0,0,0") {
    angularDampingOverride=1
    angularSpringConstant=10000
    bAngularSpring=true
    bOverrideAngularDamping=true
    bUseGravityOverride=true
    gravityScale=0.35
    simulationSpace=RootRelative
    alphaScaleBiasClamp {bias=1, scale=-1}
    boundBone {boneName=ear_r}
    componentAppliedLinearAccClamp=<10000 10000 10000>
    componentLinearAccScale=<1 1 1>
    componentLinearVelScale=<1 1 1>
    gravityOverride {z=-980}
    physicsBodyDefinitions [{
      boundBone {boneName=ear_r}
      boxExtents=<20 30 30>
      constraintSetup {angularLimitsMax=<5 10 5>, angularLimitsMin=<-5 -10 -5>, angularTarget=<1 0 0>}
      localJointOffset=<4 0 0>
     }]
   }
   n9 = Anim Dynamics - Bone: antenna(ComponentPose=@n8, Alpha=1, ExternalForce="0,0,0") {
    angularDampingOverride=1
    angularSpringConstant=20000
    bAngularSpring=true
    bOverrideAngularDamping=true
    bUseGravityOverride=true
    gravityScale=0.35
    simulationSpace=RootRelative
    alphaScaleBiasClamp {bias=1, scale=-1}
    boundBone {boneName=antenna}
    componentAppliedLinearAccClamp=<10000 10000 10000>
    componentLinearAccScale=<1 1 1>
    componentLinearVelScale=<1 1 1>
    gravityOverride {z=-980}
    physicsBodyDefinitions [{
      boundBone {boneName=antenna}
      boxExtents=<10 18 18>
      constraintSetup {angularLimitsMax=<15 15 15>, angularLimitsMin=<-15 -15 -15>, angularTarget=<1 0 0>}
      localJointOffset=<4 0 0>
     }]
   }
   n10 = "Transform (Modify) Bone - Bone: upperarm_l"!(ComponentPose=@n9, bAlphaBoolEnabled:bool=True, Rotation="12,0,0") -> (Pose) {alphaInputType=Bool, rotationMode=BMM_Additive, boneToModify {boneName=upperarm_l}, rotation {pitch=12}}
   Output Pose(Result=Component To Local(ComponentPose="Transform (Modify) Bone - Bone: upperarm_r"!(ComponentPose=@n10, bAlphaBoolEnabled:bool=True, Rotation="-12,0,0") -> (Pose) {alphaInputType=Bool, rotationMode=BMM_Additive, boneToModify {boneName=upperarm_r}, rotation {pitch=-12}}))
   Save cached pose 'StartPose'(Pose=Input Pose()) {cacheName=StartPose}
  ]}
 EventGraph {nodes [
   Event BlueprintInitializeAnimation()
   > Set mesh component(mesh component:Skeletal Mesh Component Object Reference=GetOwningComponent()) -> (Output_Get:Skeletal Mesh Component Object Reference)
   Event BlueprintUpdateAnimation()
   > Set isAccessoryON(isAccessoryON:bool=HasVariantTagForChannel!(
     self:Fort Player Companion Anim Instance Object Reference=GetMainAnimInstance!(self:Companion Post Process Anim Instance Object Reference) -> (ReturnValue:Fort Player Companion Anim Instance Object Reference)
     ChannelTag:Gameplay Tag Structure=(TagName="Cosmetics.Variant.Channel.Outfit")
     VariantTag:Gameplay Tag Structure=(TagName="Cosmetics.Variant.Property.Mat2")
    ) -> (ReturnValue:bool)) -> (Output_Get:bool)
   > n1 = Branch(Condition="float > float"(A=GetCurveValue(CurveName=isemote)))
   > Set isEmote(isEmote:bool=true) -> (Output_Get:bool)
   Set isEmote(execute=@n1.else, isEmote=false)
   > n2 = Branch(Condition="float > float"(A=GetCurveValue(CurveName=isReactDanceLoop)))
   > Set isReactDanceLoop(isReactDanceLoop:bool=true) -> (Output_Get:bool)
   Set isReactDanceLoop(execute=@n2.else, isReactDanceLoop=false)
  ]}
]
variables [
 Custom Idle Is Playing Bool {type=boolean}
 Custom Idle Is Playing Float {type=number}
 isAccessoryON {type=boolean}
 isEmote {type=boolean}
 isReactDanceLoop {type=boolean}
 mesh component {type=SkeletalMeshComponent}
 PoseAsset {type=PoseAsset, default=&/VastGeese_Biped_Pose_Asset}
]
